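# Sound cue for a magic damage burst spot sound, filed under the MephistoThrone environment actor.
# A random node picks one of three equally weighted damage flare waves, which feeds a mixer at reduced volume.
# It borrows the Disguise gadget equip attenuation and the gadget item sound class.
# A concurrency override allows one instance per owner and blocks new plays while it is active.
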SoundCue /Game/Sounds/Helios/Environment_Actors/MephistoThrone/Spot_Magic_Damage_Burst_Cue
root=Mixer_2
nodes [
 {id=Mixer_2, children [Random_0], properties {inputVolume [0.6]}}
 {id=Random_0, children [WavePlayer_0,WavePlayer_3,WavePlayer_4], properties {weights [1,1,1]}}
 {id=WavePlayer_0, properties {soundWaveAssetPtr=&/ElbowChat_Damage_Flare_01}}
 {id=WavePlayer_3, properties {soundWaveAssetPtr=&/ElbowChat_Damage_Flare_02}}
 {id=WavePlayer_4, properties {soundWaveAssetPtr=&/ElbowChat_Damage_Flare_03}}
]
properties {
 attenuationSettings=/Game/Athena/Sounds/Gadgets/Disguise/Disguise_Equip_Att
 bOverrideConcurrency=true
 maxDistance=1712
 soundClassObject=&Fort_Gadgets_Items
 volumeMultiplier=1
 concurrencyOverrides {bLimitToOwner=true, maxCount=1, resolutionRule=PreventNew}
}
